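# Translucent, lit general smoke material for sprite particles in the Galileo gameplay cue prototype, using per-vertex directional volumetric lighting.
# World normals blend a spherical normal from object center with a flattened SubUV normal map.
# A camera fade and density-modulated depth fade drive shadow-to-highlight base color and, multiplied by density, opacity.
# Particle color tints base color and emissive; dynamic parameters control normal flattening, final depth fade and base color scale.
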Material /FNGameplayCues/Athena/Prototype/Blueprints/Galileo/Materials/M_GeneralSmoke
graph {
 properties {
  bTangentSpaceNormal=false
  bUseMaterialAttributes=true
  bUsedWithNiagaraSprites=true
  bUsedWithParticleSprites=true
  blendMode=BLEND_Translucent
  pixelDepthOffsetMode=PDOM_Legacy
  translucencyLightingMode=TLM_VolumetricPerVertexDirectional
  translucentShadowStartOffset=0
  usedShadingModels=Default Lit
 }
 nodes [
  Constant() {r=1}
  n1 = WorldPosition()
  n2 = ObjectPositionWS()
  n3 = Transform(None=Normalize(VectorInput=Subtract(A=@n1.XYZ, B=@n2))) {transformSourceType=TRANSFORMSOURCE_World, transformType=TRANSFORM_Tangent}
  n4 = TextureSampleParameterSubUV!(UVs, Apply View MipBias) {parameterName=NormalMap, samplerType=SAMPLERTYPE_Normal, texture=/Game/Effects/Fort_Effects/Textures/Smoke/SmokeTest_N}
  n5 = DynamicParameter() {paramNames [FadeDistance, FlattenNormal, Param3, "ParitcleColorAffectBaseColor(0-1)"]}
  n6 = Transform(None=MaterialFunctionCall(
    BaseNormal=VertexInterpolator(VS=@n3).PS
    AdditionalNormal=MaterialFunctionCall(Normal=@n4.RGB, Flatness=@n5.FlattenNormal) {materialFunction=&FlattenNormal}.Result
   ) {materialFunction=&BlendAngleCorrectedNormals}.Result)
  n7 = ParticlePositionWS()
  n8 = WorldPosition()
  n9 = OneMinus(None=SphereMask(
    A=ShadingPathSwitch(Default=@n7, Mobile=@n8.XYZ) {desc=Migrated from FeatureLevelSwitch}
    B=CameraPositionWS()
    Radius=ScalarParameter() {defaultValue=40, parameterName=CameraFadeDistance}
   ) {hardnessPercent=0})
  n10 = ScalarParameter() {defaultValue=100, parameterName=DF_FD_Max}
  n11 = ScalarParameter() {defaultValue=20, parameterName=DF_FD_Min}
  n12 = TextureSampleParameterSubUV() {parameterName=DensityMap, samplerType=SAMPLERTYPE_LinearColor, texture=/Game/Effects/Fort_Effects/Textures/Smoke/SmokeTest2_Soft}
  n13 = ShadingPathSwitch(Default=DepthFade(Opacity=VertexInterpolator(VS=@n9).PS, FadeDistance=LinearInterpolate(A=@n10, B=@n11, Alpha=@n12.G))) {desc=Migrated from FeatureLevelSwitch}
  n14 = SetMaterialAttributes!(
   MaterialAttributes=GetMaterialAttributes(None=SetMaterialAttributes!(MaterialAttributes, Normal=@n6) {attributeSetTypes [#1]}) {attributeGetTypes [#1]}.MaterialAttributes
   Base Color=LinearInterpolate(
    A=VectorParameter() {parameterName=BaseColor_Shadow, defaultValue=<rgba 0 0 0 1>}.RGB
    B=VectorParameter() {parameterName=BaseColor_HightLight, defaultValue=<rgba 1 1 1 1>}.RGB
    Alpha=@n13
   )
  ) {attributeSetTypes [#2]}
  n15 = Multiply(A=@n12.G, B=@n13)
  n16 = GetMaterialAttributes(None=SetMaterialAttributes!(
    MaterialAttributes=SetMaterialAttributes!(MaterialAttributes=@n14, Opacity=@n15) {attributeSetTypes [#3]}
    Emissive Color=VectorParameter() {parameterName=Emissive}.RGB
   ) {attributeSetTypes [#4]}) {attributeGetTypes [#3, #4]}
  n17 = ParticleColor()
  n18 = Multiply(A=@n16.Opacity, B=@n17.A)
  n19 = DynamicParameter() {paramNames [FadeDistance, FlattenNormal, Param3, "ParitcleColorAffectBaseColor(0-1)"]}
  n20 = GetMaterialAttributes(None=SetMaterialAttributes!(
    MaterialAttributes=@n16.MaterialAttributes
    Opacity=DepthFade(Opacity=@n18, FadeDistance=@n19.FadeDistance)
    Emissive Color=Multiply(A=@n16.Emissive Color, B=@n17.RGB)
   ) {attributeSetTypes [#3, #4]}) {attributeGetTypes [#2]}
  n21 = DynamicParameter() {paramNames [FadeDistance, FlattenNormal, Param3, "ParitcleColorAffectBaseColor(0-1)"]}
  SetMaterialAttributes!(
   MaterialAttributes=@n20.MaterialAttributes
   Base Color=Multiply(A=Multiply(A=@n20.Base Color, B=@n21."ParitcleColorAffectBaseColor(0-1)"), B=ParticleColor().RGB)
  ) {attributeSetTypes [#2]}
  TextureSample() {samplerType=SAMPLERTYPE_Normal, texture=&SmokeTest_N}
  TextureSample() {samplerType=SAMPLERTYPE_LinearColor, texture=&SmokeTest2_Soft}
 ]
}
